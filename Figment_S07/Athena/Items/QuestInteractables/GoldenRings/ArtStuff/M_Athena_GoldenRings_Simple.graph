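# Simple unlit material for the Golden Rings quest interactable, flagged for use on mesh particles.
# A gold ColorInner vector parameter is multiplied by an OverallBrightness scalar to give a flat, overbright color.
# It uses no textures, and both color and brightness are exposed as parameters for instances.
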Material /Figment_S07/Athena/Items/QuestInteractables/GoldenRings/ArtStuff/M_Athena_GoldenRings_Simple
graph {
 properties {bUsedWithMeshParticles=true, pixelDepthOffsetMode=PDOM_Legacy, shadingModel=MSM_Unlit, usedShadingModels=Unlit}
 nodes [Multiply(
   A=VectorParameter() {parameterName=ColorInner, defaultValue=<rgba 0.807 0.7678 0.1705 1>}.RGB
   B=ScalarParameter() {defaultValue=8, parameterName=OverallBrightness}
  )]
}
